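# Unlit, two-sided translucent material for a shield trail effect on the Ashton Chicago backpack, built for beam trails and Niagara ribbons.
# Particle color drives emissive, while particle alpha, UV gradients, sphere masks and two panning smoke textures shape the opacity mask, with optional depth fade.
# At High quality, a distortion normal and mask-driven index of refraction add warping.
# A static switch enables vertex offsets that tilt and widen the ribbon edges.
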Material /Game/Athena/Items/Gameplay/BackPacks/Ashton/Chicago/FX_Assets/M_Ashton_Chicago_ShieldTrail
graph {
 properties {
  bUsedWithBeamTrails=true
  bUsedWithNiagaraRibbons=true
  blendMode=BLEND_Translucent
  pixelDepthOffsetMode=PDOM_Legacy
  refractionMethod=RM_IndexOfRefraction
  shadingModel=MSM_Unlit
  twoSided=true
  usedShadingModels=Unlit
 }
 nodes [
  MaterialFunctionCall(Emissive=ParticleColor().RGB) {materialFunction=&MF_FX_EmissiveAdjustments}
  QualitySwitch(
   Default=Constant()
   High=TextureSample(UVs=TextureCoordinate() {coordinateIndex=1, uTiling=0.5, vTiling=3.4}) {samplerType=SAMPLERTYPE_Normal, texture=&T_Weap_Distort_N_02}.RGB
  )
  n1 = ComponentMask(None=TextureCoordinate()) {r=true}
  n2 = Sine(None=Multiply(A=@n1) {constB=0.5})
  n3 = OneMinus(None=@n1)
  n4 = ScalarParameter() {defaultValue=2, parameterName=Mask_Power}
  n5 = ParticleColor()
  n6 = OneMinus(None=MaterialFunctionCall(Float2=TextureCoordinate()) {materialFunction=&BreakOutFloat2Components}.R)
  n7 = TextureCoordinate()
  n8 = Constant2Vector() {g=0.7, r=0.5}
  n9 = Multiply(
   A=Subtract(A=Multiply(A=@n6), B=SphereMask(A=@n7, B=@n8) {attenuationRadius=0.85, hardnessPercent=0})
   B=SphereMask(A=TextureCoordinate(), B=Constant2Vector() {g=0.5, r=1.9}) {attenuationRadius=2, hardnessPercent=95}
  )
  n10 = SphereMask(A=TextureCoordinate() {uTiling=0.3}, B=Constant2Vector() {g=0.5}) {attenuationRadius=0.5, hardnessPercent=70}
  n11 = MaterialFunctionCall(UVs=TextureCoordinate() {coordinateIndex=1}, Texture Scale=Constant2Vector() {g=1.5, r=1}) {materialFunction=&ScaleUVsByCenter}
  n12 = AppendVector(A=ScalarParameter() {defaultValue=1, parameterName=NoiseUtliing}, B=ScalarParameter() {defaultValue=1, parameterName=NoiseVtliing})
  n13 = ScalarParameter() {defaultValue=0.5, parameterName=PanSpeedU}
  n14 = ScalarParameter() {parameterName=PanSpeedV}
  n15 = TextureSample(UVs=Panner(Coordinate=Multiply(A=@n11.UVs, B=@n12), Speed=AppendVector(A=@n13, B=@n14)) {speedX=0.5}) {samplerType=SAMPLERTYPE_Grayscale, texture=&T_Smoke_Trail_Soft}
  n16 = TextureCoordinate() {coordinateIndex=1, uTiling=0.35, vTiling=1.5}
  n17 = Constant2Vector() {g=1.5, r=1}
  n18 = ScalarParameter() {defaultValue=1, parameterName=NoiseUtliing}
  n19 = ScalarParameter() {defaultValue=1, parameterName=NoiseVtliing}
  n20 = Panner(
   Coordinate=Multiply(
    A=MaterialFunctionCall(UVs=@n16, Texture Scale=@n17) {materialFunction=&ScaleUVsByCenter}.UVs
    B=AppendVector(A=@n18, B=@n19)
   )
   Speed=AppendVector(A=ScalarParameter() {defaultValue=0.5, parameterName=PanSpeedU}, B=ScalarParameter() {parameterName=PanSpeedV})
  ) {speedX=0.5}
  n21 = Multiply(A=Power(Base=TextureSample(UVs=@n20) {texture=&T_Inky_Smoke_Tile_Chromatic}.RGB)) {constB=5}
  n22 = Add(A=Saturate(None=Multiply(A=@n9, B=@n10)), B=Multiply(A=Add(A=@n15.RGB, B=@n21)) {constB=0.8})
  n23 = Multiply(A=Saturate(None=Multiply(A=@n2) {constB=1.5}), B=Multiply(A=Power(Base=@n3, Exp=@n4), B=Multiply(A=@n5.A, B=@n22)))
  n24 = DepthFade(FadeDistance=ScalarParameter() {parameterName=Depth Fade Diatnce})
  n25 = Multiply(
   A=LinearInterpolate(A=@n23, B=Multiply(A=@n24, B=@n23), Alpha=ScalarParameter() {parameterName=Use Depth Fade})
   B=SphereMask(A=TextureCoordinate() {uTiling=0.5}, B=Constant2Vector() {g=0.5}) {attenuationRadius=0.5, hardnessPercent=0}
  )
  QualitySwitch(Default=Constant() {r=1}, High=LinearInterpolate(Alpha=ComponentMask(None=@n25) {r=true}) {constA=1, constB=1.2})
  n26 = ScalarParameter() {defaultValue=0.9, parameterName=Width}
  n27 = Multiply(A=VectorParameter() {parameterName=RibbonTiltTop, defaultValue=<rgba -50 0 -50 0>}.RGB, B=@n26)
  n28 = Multiply(A=@n26, B=VectorParameter() {parameterName=RibbonTiltBottom, defaultValue=<rgba 50 0 50 0>}.RGB)
  n29 = Round(None=ComponentMask(None=TextureCoordinate()) {g=true})
  StaticSwitchParameter(True=LinearInterpolate(A=Transform(None=@n27), B=Transform(None=@n28), Alpha=Saturate(None=@n29)), False=Constant()) {parameterName=Use Horizontal Function}
 ]
}
